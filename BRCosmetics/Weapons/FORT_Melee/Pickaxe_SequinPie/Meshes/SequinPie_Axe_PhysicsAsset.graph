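# Physics asset for the SequinPie two-handed pickaxe cosmetic, defining the collision shape of its axe skeletal mesh.
# A single body on the Melee_TwoHanded_Handle bone uses one capsule, offset up the handle and rotated to align with it.
# It has no constraints, so the axe simulates and collides as one rigid piece.
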PhysicsAsset /BRCosmetics/Weapons/FORT_Melee/Pickaxe_SequinPie/Meshes/SequinPie_Axe_PhysicsAsset
bodies {Melee_TwoHanded_Handle [{shapeType=Capsule, length=51.3322, radius=5.9951, center=<-0.0148 1.6882 36.3125>, rotation=<rot 4.1298 -90 -0.0363>}]}
constraints []
